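# Additive unlit, two-sided Niagara material for the Bistro Astronaut female glider's trail.
# A panning circuit-board texture and UV sphere-mask falloffs form the body, with a UseLine blend adding an edge line that varies along U.
# A curve-atlas gradient along U, blended by particle color, is converted to HDR using that shape.
# Use_FoeGlow lerps between a Fresnel Foe_Color glow at 0 and this trail at 1, the default.
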Material /BRCosmetics/Gadgets/Assets/VinderTech_GliderChute/Glider_Bistro_Astronaut_Female/FX/M_Bistro_Astronaut_Trail
graph {
 properties {
  bUsedWithNiagaraMeshParticles=true
  bUsedWithNiagaraRibbons=true
  bUsedWithNiagaraSprites=true
  blendMode=BLEND_Additive
  pixelDepthOffsetMode=PDOM_Legacy
  shadingModel=MSM_Unlit
  twoSided=true
  usedShadingModels=Unlit
 }
 nodes [
  Constant3Vector() {constant=<rgba 0.0306 1 0 0>}
  n1 = OneMinus(None=Fresnel(ExponentIn=ScalarParameter() {defaultValue=1}, Normal=VertexNormalWS()) {baseReflectFraction=0})
  n2 = ParticleColor()
  n3 = ParticleColor()
  n4 = SphereMask(A=TextureCoordinate() {uTiling=0.3}, B=Constant2Vector() {g=0.5, r=1}) {attenuationRadius=1, hardnessPercent=0}
  n5 = Power(Base=OneMinus(None=Multiply(A=@n4) {constB=5}))
  n6 = AppendVector(
   A=ScalarParameter() {defaultValue=3, parameterName=TextureA_U Tiling}
   B=ScalarParameter() {defaultValue=0.5, parameterName=TextureA_VTiling}
  )
  n7 = TextureCoordinate()
  n8 = Panner(Coordinate=MaterialFunctionCall(XY=Multiply(A=@n6, B=@n7)) {materialFunction=&Swizzle}.YX) {speedY=-3}
  n9 = TextureSampleParameter2D(UVs=@n8) {parameterName=Texture A, texture=/Game/Effects/Fort_Effects/Textures/T_FX_Circuitboard}
  n10 = Constant() {r=10}
  n11 = TextureCoordinate()
  n12 = OneMinus(None=ComponentMask(None=@n11) {g=true})
  n13 = ComponentMask(None=@n11) {r=true}
  n14 = SphereMask(A=@n13, B=Constant() {r=-0.1}) {attenuationRadius=1, hardnessPercent=0}
  n15 = Multiply(A=Add(A=Multiply(A=@n5) {constB=0.55}, B=Multiply(A=@n9.R, B=@n10)), B=Multiply(A=Multiply(A=@n12, B=@n12), B=Multiply(A=@n14, B=@n14)))
  n16 = TextureCoordinate()
  n17 = OneMinus(None=SphereMask(A=ComponentMask(None=@n16) {g=true}, B=Constant() {r=-0.3}) {attenuationRadius=0.7, hardnessPercent=0})
  n18 = Constant() {r=0.01}
  n19 = AppendVector(A=@n18, B=ComponentMask(None=TextureCoordinate()) {g=true})
  n20 = ScalarParameter() {defaultValue=0.05, group=Hard Line, parameterName=Edge RadiusAltA, sortPriority=1}
  n21 = ScalarParameter() {defaultValue=0.01, group=Hard Line, parameterName=Edge RadiusAltB, sortPriority=2}
  n22 = ComponentMask(None=TextureCoordinate()) {r=true}
  n23 = LinearInterpolate(
   A=ScalarParameter() {defaultValue=0.3, group=Hard Line, parameterName=Edge Hardness A, sortPriority=3}
   B=ScalarParameter() {group=Hard Line, parameterName=Edge Hardness B, sortPriority=4}
   Alpha=@n22
  )
  n24 = ScalarParameter() {defaultValue=0.02, group=Hard Line, parameterName=Edge Radius A, sortPriority=1}
  n25 = ScalarParameter() {defaultValue=0.005, group=Hard Line, parameterName=Edge Radius B, sortPriority=2}
  n26 = LinearInterpolate(
   A=SphereMask(A=@n19, B=@n18, Radius=LinearInterpolate(A=@n20, B=@n21, Alpha=@n22), Hardness=@n23)
   B=SphereMask(A=@n19, B=@n18, Radius=LinearInterpolate(A=@n24, B=@n25, Alpha=@n22), Hardness=@n23)
   Alpha=ScalarParameter() {defaultValue=1, parameterName=UseAltLine}
  )
  n27 = LinearInterpolate(
   A=Multiply(A=@n15, B=Power(Base=@n17) {constExponent=5})
   B=Add(A=@n15, B=Multiply(A=@n26) {constB=40})
   Alpha=ScalarParameter() {defaultValue=1, parameterName=UseLine}
  )
  LinearInterpolate(
   A=DepthFade(Opacity=Multiply(A=@n1, B=@n2.A), FadeDistance=ScalarParameter() {defaultValue=0.5, parameterName=Param_1}) {fadeDistanceDefault=3}
   B=DepthFade(Opacity=Multiply(A=@n3.A, B=@n27))
   Alpha=ScalarParameter() {defaultValue=1, parameterName=Use_FoeGlow}
  )
  n28 = VectorParameter() {parameterName=Foe_Color, defaultValue=<rgba 0 1 0.2152 1>}
  n29 = Clamp(None=@n28.RGB)
  n30 = ScalarParameter() {defaultValue=50, parameterName=Light}
  n31 = CurveAtlasRowParameter(CurveTime=@n13) {
   atlas=&AthenaCurveAtlas01
   curve=/Game/Gadgets/Assets/VinderTech_GliderChute/Glider_Bistro_Astronaut_Female/FX/Curve_Color_Bistro_Trail
   defaultValue=305
   group=Color Curve
   parameterName=ColorCurve
  }
  n32 = LinearInterpolate(A=Constant() {r=1}, B=@n31, Alpha=ParticleColor().RGB) {constA=1}
  n33 = Desaturation(None=@n32, Fraction=Constant() {r=-0.5})
  n34 = Constant() {r=100}
  n35 = Clamp(None=@n32)
  MaterialFunctionCall(Emissive=LinearInterpolate(
    A=MaterialFunctionCall(LDR Input=@n29, HDR Tint=@n28.RGB, HDR Intensity Multiplier=@n30) {materialFunction=&DeriveHDRfromLDR}.RemappedRange
    B=MaterialFunctionCall(LDR Input=@n33, HDR Dynamic Range=@n34, HDR Tint=@n35, HDR Intensity Multiplier=@n27) {materialFunction=&DeriveHDRfromLDR}.RemappedRange
    Alpha=ScalarParameter() {defaultValue=1, parameterName=Use_FoeGlow}
   )) {materialFunction=&MF_FX_EmissiveAdjustments}
  Multiply(A=@n31, B=@n32)
  TextureSample(UVs=@n8) {texture=/BRCosmetics/Weapons/FORT_Melee/Pickaxe_Bistro_Astronaut_Female/FX/T_Pickaxe_Bistro}
 ]
}
